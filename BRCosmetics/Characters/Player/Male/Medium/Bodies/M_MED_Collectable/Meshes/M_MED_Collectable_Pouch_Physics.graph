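# Physics asset for the pouch on the medium male Collectable player body, apparently for secondary motion of the dyn_pouch bones against the body.
# A box body on dyn_pouch is constrained to a tiny capsule on dyn_pouch_MSTR, with a narrower twist limit than swing.
# Capsules on pelvis, spine_01 and left thigh approximate the nearby body.
# The master bone and thigh each link to the pelvis with free angular motion, while spine_01 has no constraint.
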PhysicsAsset /BRCosmetics/Characters/Player/Male/Medium/Bodies/M_MED_Collectable/Meshes/M_MED_Collectable_Pouch_Physics
bodies {
 dyn_pouch [{
   shapeType=Box
   extentX=16.4154
   extentY=11.4456
   extentZ=5.8411
   shapeName=dyn_pouch_box
   center=<1.5 -1.5 3>
   rotation=<rot -1.2924 -4.8302 -9.9456>
  }]
 dyn_pouch_MSTR [{shapeType=Capsule, length=0.505, radius=0.505, shapeName=dyn_pouch_MSTR_capsule, rotation=<rot 0 -90 0>}]
 pelvis [{shapeType=Capsule, length=7.2941, radius=11.5898, shapeName=pelvis_capsule, center=<0.7683 -1.9682 0>}]
 spine_01 [{shapeType=Capsule, length=3.8882, radius=11.1146, shapeName=spine_01_capsule, center=<0 -1.7214 0>}]
 thigh_l [{shapeType=Capsule, length=18.7431, radius=8.969, shapeName=thigh_l_capsule, center=<-18.2557 0 0>, rotation=<rot 90 0 0>}]
}
constraints [
 {bone1Name=dyn_pouch, bone2Name=dyn_pouch_MSTR, swing1LimitDegrees=45, swing2LimitDegrees=45, twistLimitDegrees=15}
 {
  bone1Name=dyn_pouch_MSTR
  bone2Name=pelvis
  swing1LimitDegrees=45
  swing1Motion=Free
  swing2LimitDegrees=45
  swing2Motion=Free
  twistLimitDegrees=45
  twistMotion=Free
 }
 {
  bone1Name=thigh_l
  bone2Name=pelvis
  swing1LimitDegrees=45
  swing1Motion=Free
  swing2LimitDegrees=45
  swing2Motion=Free
  twistLimitDegrees=45
  twistMotion=Free
 }
]
