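# Physics asset for the M_MED_Collectable male character's crouch leg mesh, providing bodies for the dynamic ropes on both calves.
# Each calf capsule anchors a master rope body and root that branch into short capsule chains, with the right side one segment shorter.
# Both master-to-calf joints and the right root-to-master joint swing and twist freely, while the other joints carry angular limits, so the two sides are not symmetric.
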PhysicsAsset /BRCosmetics/Characters/Player/Male/Medium/Bodies/M_MED_Collectable/Meshes/M_MED_Collectable_crouch_Leg_Physics
bodies {
 calf_l [{shapeType=Capsule, length=28.4539, radius=5.5555, center=<-21.0766 -1.4018 0.3607>, rotation=<rot 0 -96.0143 -89.1834>}]
 calf_r [{
   shapeType=Capsule
   length=28.4539
   radius=5.5555
   shapeName=calf_r_capsule
   center=<21.0192 -1.9468 0.3607>
   rotation=<rot 0 -96.0143 -89.1834>
  }]
 dyn_ropes_l [{shapeType=Capsule, length=0.505, radius=0.505, shapeName=dyn_ropes_l_capsule}]
 dyn_ropes_l_MSTR [{shapeType=Capsule, length=6.4238, radius=9.5003, shapeName=dyn_ropes_l_MSTR_capsule, center=<3.2301 -0.4209 -9.5775>, rotation=<rot 0 0 90>}]
 dyn_ropes_l_a [{shapeType=Capsule, length=4.1742, radius=1.1763, shapeName=dyn_ropes_l_a_capsule, center=<4 0 0>, rotation=<rot 90 0 0>}]
 dyn_ropes_l_a_01 [{shapeType=Capsule, length=8.6869, radius=1.2702, shapeName=dyn_ropes_l_a_01_capsule, center=<5.5 0 0>, rotation=<rot 90 0 0>}]
 dyn_ropes_l_b [{shapeType=Capsule, length=4.2459, radius=1.1398, shapeName=dyn_ropes_l_b_capsule, center=<4 0 0>, rotation=<rot 90 0 0>}]
 dyn_ropes_l_b_01 [{shapeType=Capsule, length=7.1805, radius=1.1641, shapeName=dyn_ropes_l_b_01_capsule, center=<5 0 0>, rotation=<rot 90 0 0>}]
 dyn_ropes_r [{shapeType=Capsule, length=0.505, radius=0.505, shapeName=dyn_ropes_r_capsule}]
 dyn_ropes_r_MSTR [{shapeType=Capsule, length=9.8571, radius=6.067, shapeName=dyn_ropes_r_MSTR_capsule, center=<3.7779 -0.4209 -5.7931>, rotation=<rot 0 0 90>}]
 dyn_ropes_r_a [{shapeType=Capsule, length=4.3501, radius=0.8733, shapeName=dyn_ropes_r_a_capsule, center=<3.0721 0 0>, rotation=<rot 90 0 0>}]
 dyn_ropes_r_a_01 [{shapeType=Capsule, length=7.9128, radius=1.1149, shapeName=dyn_ropes_r_a_01_capsule, center=<5.3904 0 0>, rotation=<rot 0 -90 -89.5>}]
 dyn_ropes_r_b [{shapeType=Capsule, length=7.2933, radius=0.8155, shapeName=dyn_ropes_r_b_capsule, center=<5.2186 0 0>, rotation=<rot 0 -85.7324 -88.8>}]
}
constraints [
 {bone1Name=dyn_ropes_l, bone2Name=dyn_ropes_l_MSTR, swing1LimitDegrees=45, swing2LimitDegrees=45, twistLimitDegrees=45}
 {
  bone1Name=dyn_ropes_l_MSTR
  bone2Name=calf_l
  swing1LimitDegrees=45
  swing1Motion=Free
  swing2LimitDegrees=45
  swing2Motion=Free
  twistLimitDegrees=45
  twistMotion=Free
 }
 {bone1Name=dyn_ropes_l_a_01, bone2Name=dyn_ropes_l_a, swing1LimitDegrees=45, swing2LimitDegrees=45, twistLimitDegrees=15}
 {bone1Name=dyn_ropes_l_b, bone2Name=dyn_ropes_l, swing1LimitDegrees=45, swing2LimitDegrees=10, twistLimitDegrees=10}
 {bone1Name=dyn_ropes_l_b_01, bone2Name=dyn_ropes_l_b, swing1LimitDegrees=45, swing2LimitDegrees=45, twistLimitDegrees=15}
 {bone1Name=dyn_ropes_l_a, bone2Name=dyn_ropes_l, swing1LimitDegrees=45, swing2LimitDegrees=10, twistLimitDegrees=10}
 {bone1Name=dyn_ropes_r_a_01, bone2Name=dyn_ropes_r_a, swing1LimitDegrees=45, swing2LimitDegrees=45, twistLimitDegrees=15}
 {bone1Name=dyn_ropes_r_b, bone2Name=dyn_ropes_r, swing1LimitDegrees=45, swing2LimitDegrees=45, twistLimitDegrees=15}
 {bone1Name=dyn_ropes_r_a, bone2Name=dyn_ropes_r, swing1LimitDegrees=45, swing2LimitDegrees=45, twistLimitDegrees=15}
 {
  bone1Name=dyn_ropes_r
  bone2Name=dyn_ropes_r_MSTR
  swing1LimitDegrees=45
  swing1Motion=Free
  swing2LimitDegrees=45
  swing2Motion=Free
  twistLimitDegrees=45
  twistMotion=Free
 }
 {
  bone1Name=dyn_ropes_r_MSTR
  bone2Name=calf_r
  swing1LimitDegrees=45
  swing1Motion=Free
  swing2LimitDegrees=45
  swing2Motion=Free
  twistLimitDegrees=45
  twistMotion=Free
 }
]
