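# Character-part AnimBP for the M_MED_Collectable player body, adding secondary motion to its ropes, pouch, pants and other dyn_ bones.
# It copies the attached parent's pose, then poses helper bones with weighted constraints, look-ats and offsets driven by calf bend and crouch.
# That cached pose feeds separate standing and crouching leg RigidBody sims, blended by bIsCrouching, then a pouch sim.
# Several AnimDynamics nodes sit unconnected to the output.
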AnimBlueprint /BRCosmetics/Characters/Player/Male/Medium/Bodies/M_MED_Collectable/Meshes/M_MED_Collectable_AnimBP
target_skeleton=&/M_MED_Collectable_Skeleton
graphs [
 AnimGraph {nodes [
   n1 = Anim Dynamics - Bone: dyn_ropes_l_a(Alpha=1, ExternalForce="0,0,0") {
    angularDampingOverride=0.9999
    angularSpringConstant=50
    bAngularSpring=true
    bGravityOverrideInSimSpace=true
    bOverrideAngularDamping=true
    simulationSpace=BoneRelative
    alphaBoolBlend {blendOption=Linear}
    boundBone {boneName=dyn_ropes_l_a}
    componentAppliedLinearAccClamp=<3500 3500 3500>
    componentLinearAccScale=<1.2 1.2 1.2>
    componentLinearVelScale=<1.2 1.2 1.2>
    physicsBodyDefinitions [{
      sphereCollisionRadius=0
      boundBone {boneName=dyn_ropes_l_a}
      boxExtents=<12 7 12>
      constraintSetup {angularLimitsMax=<1 30 -5>, angularLimitsMin=<-1 -30 -35>, angularTarget=<0 0 0>}
      localJointOffset=<2 0 -1>
     }]
    relativeSpaceBone {boneName=thigh_l}
   }
   n2 = Anim Dynamics - Bone: dyn_ropes_l_a(ComponentPose=@n1, Alpha=1, ExternalForce="0,0,0") {
    angularDampingOverride=0.9999
    angularSpringConstant=50
    bAngularSpring=true
    bGravityOverrideInSimSpace=true
    bOverrideAngularDamping=true
    simulationSpace=BoneRelative
    alphaBoolBlend {blendOption=Linear}
    boundBone {boneName=dyn_ropes_l_a}
    componentAppliedLinearAccClamp=<3500 3500 3500>
    componentLinearAccScale=<1.2 1.2 1.2>
    componentLinearVelScale=<1.2 1.2 1.2>
    physicsBodyDefinitions [{
      sphereCollisionRadius=0
      boundBone {boneName=dyn_ropes_l_a}
      boxExtents=<12 7 12>
      constraintSetup {angularLimitsMax=<1 30 5>, angularLimitsMin=<-1 -30 -35>, angularTarget=<0 0 0>}
      localJointOffset=<2 0 -1>
     }]
    relativeSpaceBone {boneName=thigh_l}
   }
   n3 = Anim Dynamics - Bone: dyn_ropes_l_b(ComponentPose=@n2, Alpha=1, ExternalForce="0,0,0") {
    angularDampingOverride=0.9999
    angularSpringConstant=100
    bAngularSpring=true
    bGravityOverrideInSimSpace=true
    bOverrideAngularDamping=true
    simulationSpace=BoneRelative
    alphaBoolBlend {blendOption=Linear}
    boundBone {boneName=dyn_ropes_l_b}
    componentAppliedLinearAccClamp=<3500 3500 3500>
    componentLinearAccScale=<1.2 1.2 1.2>
    componentLinearVelScale=<1.2 1.2 1.2>
    physicsBodyDefinitions [{
      sphereCollisionRadius=0
      boundBone {boneName=dyn_ropes_l_b}
      boxExtents=<12 7 12>
      constraintSetup {angularLimitsMax=<1 30 -5>, angularLimitsMin=<-1 -30 -35>, angularTarget=<0 0 0>}
      localJointOffset=<2 0 0>
     }]
    relativeSpaceBone {boneName=thigh_l}
   }
   Anim Dynamics - Bone: dyn_ropes_l_b_01(ComponentPose=@n3, Alpha=1, ExternalForce="0,0,0") {
    angularDampingOverride=0.9999
    angularSpringConstant=100
    bAngularSpring=true
    bGravityOverrideInSimSpace=true
    bOverrideAngularDamping=true
    simulationSpace=BoneRelative
    alphaBoolBlend {blendOption=Linear}
    boundBone {boneName=dyn_ropes_l_b_01}
    componentAppliedLinearAccClamp=<3500 3500 3500>
    componentLinearAccScale=<1.2 1.2 1.2>
    componentLinearVelScale=<1.2 1.2 1.2>
    physicsBodyDefinitions [{
      sphereCollisionRadius=0
      boundBone {boneName=dyn_ropes_l_b_01}
      boxExtents=<12 7 12>
      constraintSetup {angularLimitsMax=<1 30 5>, angularLimitsMin=<-1 -30 -35>, angularTarget=<0 0 0>}
      localJointOffset=<2 0 0.5>
     }]
    relativeSpaceBone {boneName=thigh_l}
   }
   Anim Dynamics - Bone: dyn_pouch_01(Alpha=1, ExternalForce="0,0,0") {
    angularDampingOverride=0.998
    angularSpringConstant=10000
    bAngularSpring=true
    bOverrideAngularDamping=true
    bUseGravityOverride=true
    simulationSpace=RootRelative
    alphaBoolBlend {blendOption=Linear}
    boundBone {boneName=dyn_pouch_01}
    componentLinearAccScale=<0.8 0.8 0.8>
    componentLinearVelScale=<0.8 0.8 0.8>
    gravityOverride {z=-980}
    physicsBodyDefinitions [{
      collisionType=CustomSphere
      sphereCollisionRadius=5
      boundBone {boneName=dyn_pouch_01}
      boxExtents=<7 12 12>
      constraintSetup {angularLimitsMax=<1 3 10>, angularLimitsMin=<-1 -15 -10>, angularTarget=<0 0 0>}
      localJointOffset=<1 0 0>
     }]
    sphericalLimits [{limitRadius=9, drivingBone {boneName=thigh_l}, sphereLocalOffset=<-5 0 0>}]
   }
   Anim Dynamics - Bone: dyn_pouch_01(Alpha=1, ExternalForce="0,0,0") {
    angularDampingOverride=0.998
    angularSpringConstant=10000
    bAngularSpring=true
    bOverrideAngularDamping=true
    bUseGravityOverride=true
    simulationSpace=RootRelative
    alphaBoolBlend {blendOption=Linear}
    boundBone {boneName=dyn_pouch_01}
    componentLinearAccScale=<0.8 0.8 0.8>
    componentLinearVelScale=<0.8 0.8 0.8>
    gravityOverride {z=-980}
    physicsBodyDefinitions [{
      collisionType=CustomSphere
      sphereCollisionRadius=5
      boundBone {boneName=dyn_pouch_01}
      boxExtents=<7 12 12>
      constraintSetup {angularLimitsMax=<1 3 10>, angularLimitsMin=<-1 -15 -10>, angularTarget=<0 0 0>}
      localJointOffset=<1 0 0>
     }]
    sphericalLimits [{limitRadius=9, drivingBone {boneName=thigh_l}, sphereLocalOffset=<-5 0 0>}]
   }
   Anim Dynamics - Bone: dyn_pouch(Alpha=1, ExternalForce="0,0,0") {
    angularDampingOverride=0.998
    angularSpringConstant=5000
    bAngularSpring=true
    bOverrideAngularDamping=true
    simulationSpace=RootRelative
    alphaBoolBlend {blendOption=Linear}
    boundBone {boneName=dyn_pouch}
    componentLinearAccScale=<0.4 0.4 0.4>
    componentLinearVelScale=<0.4 0.4 0.4>
    physicsBodyDefinitions [{
      boundBone {boneName=dyn_pouch}
      boxExtents=<15 20 20>
      constraintSetup {angularLimitsMax=<1 5 5>, angularLimitsMin=<-1 -5 -5>, angularTarget=<0 0 0>}
      localJointOffset=<1 0 0>
     }]
   }
   "Constraint - dyn_pouch_MSTR (PP)"!(ComponentPose, Alpha=1, ConstraintWeights_0=0.8484, ConstraintWeights_1:float=0.1516) -> (Pose) {
    alphaBoolBlend {blendOption=Linear}
    boneToModify {boneName=dyn_pouch_MSTR}
    constraintSetup [{targetBone {boneName=pelvis}}, {targetBone {boneName=thigh_twist_01_l}}]
    constraintWeights [1,1]
   }
   "Transform (Modify) Bone - Bone: dyn_spine_05"(Alpha=Get Arm L rot() -> (Arm L rot:double), Translation="-2,0,0", Rotation="0,0,0", Scale="1,1,1") {
    translationMode=BMM_Additive
    translationSpace=BCS_BoneSpace
    alphaBoolBlend {blendOption=Linear}
    boneToModify {boneName=dyn_spine_05}
    translation {x=-2}
   }
   n4 = RigidBody!(
    ComponentPose=Local To Component(LocalPose=Use cached pose 'pre sim'())
    Alpha
    OverrideWorldGravity:Vector=Get GravityOverride()
    ExternalForce
   ) -> (Pose) {
    bClampLinearTranslationLimitToRefPose=true
    bOverrideWorldGravity=true
    overridePhysicsAsset=&/M_MED_Collectable_Leg_Physics
    simulationSpace=BaseBoneSpace
    alphaBoolBlend {blendOption=Linear}
    baseBoneRef {boneName=root}
    componentAppliedLinearAccClamp=<300 300 300>
    componentLinearAccScale=<0.25 0.25 0.25>
    componentLinearVelScale=<0.2 0.2 0.2>
   }
   n5 = Component To Local(ComponentPose=@n4)
   n6 = "Transform (Modify) Bone - Bone: dyn_ropes_l_loop"(
    ComponentPose=Local To Component(LocalPose=Use cached pose 'pre sim'())
    Alpha=Get Calf Rot L() -> (Calf Rot L:double)
    Translation="-2.058,1.474,-0.496"
    Rotation="21.260,17.005,97.246"
    Scale="1,1,1"
   ) {
    rotationMode=BMM_Additive
    rotationSpace=BCS_BoneSpace
    translationMode=BMM_Additive
    translationSpace=BCS_BoneSpace
    alphaBoolBlend {blendOption=Linear}
    boneToModify {boneName=dyn_ropes_l_loop}
    rotation=<rot 21.26 17.005 97.246>
    translation=<-2.058 1.474 -0.496>
   }
   n7 = "Transform (Modify) Bone - Bone: dyn_ropes_r_loop_b"(
    ComponentPose=@n6
    Alpha=Get Calf Rot R() -> (Calf Rot R:double)
    Translation="-0.475,0.663,-0.404"
    Rotation="-29.551,6.689,-67.143"
    Scale="1,1,1"
   ) {
    rotationMode=BMM_Additive
    rotationSpace=BCS_BoneSpace
    translationMode=BMM_Additive
    translationSpace=BCS_BoneSpace
    alphaBoolBlend {blendOption=Linear}
    boneToModify {boneName=dyn_ropes_r_loop_b}
    rotation=<rot -29.551 6.689 -67.143>
    translation=<-0.475 0.663 -0.404>
   }
   n8 = RigidBody(ComponentPose=@n7, OverrideWorldGravity="0,0,-980") {
    bClampLinearTranslationLimitToRefPose=true
    bOverrideWorldGravity=true
    overridePhysicsAsset=&/M_MED_Collectable_crouch_Leg_Physics
    simulationSpace=BaseBoneSpace
    alphaBoolBlend {blendOption=Linear}
    baseBoneRef {boneName=pelvis}
    componentAppliedLinearAccClamp=<300 300 300>
    componentLinearAccScale=<0.25 0.25 0.25>
    componentLinearVelScale=<0.2 0.2 0.2>
   }
   n9 = Component To Local(ComponentPose=@n8)
   n10 = Get bIsCrouching()
   n11 = RigidBody(
    ComponentPose=Local To Component(LocalPose=Blend Poses by bool(BlendPose_0=@n9, BlendPose_1=@n5, bActiveValue=@n10) {blendType=Linear})
    OverrideWorldGravity="0,0,-980"
   ) {
    bClampLinearTranslationLimitToRefPose=true
    bOverrideWorldGravity=true
    overridePhysicsAsset=&/M_MED_Collectable_Pouch_Physics
    simulationSpace=BaseBoneSpace
    alphaBoolBlend {blendOption=Linear}
    baseBoneRef {boneName=pelvis}
    componentAppliedLinearAccClamp=<3000 3000 3000>
    componentLinearAccScale=<0.25 0.25 0.25>
    componentLinearVelScale=<0.2 0.2 0.2>
   }
   Output Pose(Result=Component To Local(ComponentPose=@n11)) {layerGroup=None}
   n12 = "Constraint - dyn_spine_05 (PPP)"!(
    ComponentPose=Local To Component(LocalPose=Copy Pose From Mesh() {bUseAttachedParent=true})
    Alpha=1
    ConstraintWeights_0=1
    ConstraintWeights_1:float=0
    ConstraintWeights_2:float=0
   ) -> (Pose) {
    alphaBoolBlend {blendOption=Linear}
    boneToModify {boneName=dyn_spine_05}
    constraintSetup [{targetBone {boneName=spine_05}}, {targetBone {boneName=spine_04}}, {targetBone {boneName=head}}]
    constraintWeights [1,1,1]
   }
   n13 = "Constraint - dyn_spine_05 (TT)"!(ComponentPose=@n12, Alpha=0.1, ConstraintWeights_0=0.6, ConstraintWeights_1:float=0.4) -> (Pose) {
    alphaBoolBlend {blendOption=Linear}
    boneToModify {boneName=dyn_spine_05}
    constraintSetup [{transformType=Translation, targetBone {boneName=upperarm_l}}, {transformType=Translation, targetBone {boneName=upperarm_r}}]
    constraintWeights [1,1]
   }
   n14 = "Constraint - dyn_abdomen (R)"(ComponentPose=@n13, Alpha=1, ConstraintWeights_0=0.5) {
    alphaBoolBlend {blendOption=Linear}
    boneToModify {boneName=dyn_abdomen}
    constraintSetup [{transformType=Rotation, targetBone {boneName=dyn_spine_05}}]
    constraintWeights [1]
   }
   n15 = Look At - Bone: dyn_spine_05(ComponentPose=@n14, Alpha=1) {
    alphaBoolBlend {blendOption=Linear}
    boneToModify {boneName=dyn_spine_05}
    lookAtLocation {x=0, z=1}
    lookAtTarget {boneReference {boneName=dyn_spine_TGT}}
    lookAt_Axis {axis=<1 0>}
   }
   n16 = Look At - Bone: dyn_abdomen(ComponentPose=@n15, Alpha=1) {
    alphaBoolBlend {blendOption=Linear}
    boneToModify {boneName=dyn_abdomen}
    lookAtLocation=<0 0>
    lookAtTarget {boneReference {boneName=dyn_abdomen_TGT}}
    lookAt_Axis {axis=<1 0>}
   }
   n17 = "Constraint - dyn_pouch_MSTR (PPPP)"!(
    ComponentPose=@n16
    Alpha=1
    ConstraintWeights_0=0.754
    ConstraintWeights_1:float=0.1086
    ConstraintWeights_2:float=0.0027
    ConstraintWeights_3:float=0.1347
   ) -> (Pose) {
    alphaBoolBlend {blendOption=Linear}
    boneToModify {boneName=dyn_pouch_MSTR}
    constraintSetup [{targetBone {boneName=pelvis}}, {targetBone {boneName=spine_01}}, {targetBone {boneName=thigh_l}}, {targetBone {boneName=thigh_twist_01_l}}]
    constraintWeights [1,1,1,1]
   }
   n18 = "Transform (Modify) Bone - Bone: dyn_pouch_MSTR"!(ComponentPose=@n17, bAlphaBoolEnabled:bool=Get bIsCrouching(), Translation="-1,-5,-3", Rotation="0,-35,-50", Scale="1,1,1") -> (Pose) {
    alphaInputType=Bool
    rotationMode=BMM_Additive
    rotationSpace=BCS_BoneSpace
    translationMode=BMM_Additive
    translationSpace=BCS_BoneSpace
    alphaBoolBlend {blendInTime=0.1, blendOption=Linear, blendOutTime=0.1}
    boneToModify {boneName=dyn_pouch_MSTR}
    rotation {roll=-50, yaw=-35}
    translation=<-1 -5 -3>
   }
   n19 = "Constraint - dyn_ankle_l_MSTR (PPP)"!(ComponentPose=@n18, Alpha=1, ConstraintWeights_0=0.2846, ConstraintWeights_1:float=0.0074, ConstraintWeights_2:float=0.708) -> (Pose) {
    alphaBoolBlend {blendOption=Linear}
    boneToModify {boneName=dyn_ankle_l_MSTR}
    constraintSetup [{targetBone {boneName=calf_twist_01_l}}, {targetBone {boneName=calf_twist_02_l}}, {targetBone {boneName=foot_l}}]
    constraintWeights [1,1,1]
   }
   n20 = "Constraint - dyn_ankle_r_MSTR (PPP)"!(ComponentPose=@n19, Alpha=1, ConstraintWeights_0=0.2034, ConstraintWeights_1:float=0.0097, ConstraintWeights_2:float=0.787) -> (Pose) {
    alphaBoolBlend {blendOption=Linear}
    boneToModify {boneName=dyn_ankle_r_MSTR}
    constraintSetup [{targetBone {boneName=calf_twist_01_r}}, {targetBone {boneName=calf_twist_02_r}}, {targetBone {boneName=foot_r}}]
    constraintWeights [1,1,1]
   }
   n21 = Look At - Bone: dyn_ankle_l(ComponentPose=@n20, Alpha=1) {
    alphaBoolBlend {blendOption=Linear}
    boneToModify {boneName=dyn_ankle_l}
    lookAtLocation {x=0}
    lookAtTarget {boneReference {boneName=calf_l}}
    lookAt_Axis {axis=<1 0>}
   }
   n22 = Look At - Bone: dyn_ankle_r(ComponentPose=@n21, Alpha=1) {
    alphaBoolBlend {blendOption=Linear}
    boneToModify {boneName=dyn_ankle_r}
    lookAtLocation {x=0}
    lookAtTarget {boneReference {boneName=calf_r}}
    lookAt_Axis {axis=<1 0>}
   }
   n23 = "Transform (Modify) Bone - Bone: dyn_pants_l"(ComponentPose=@n22, Alpha=Get Calf Rot L(), Translation="-13,8,0", Rotation="0,0,0", Scale="1,1,1") {
    translationMode=BMM_Additive
    translationSpace=BCS_BoneSpace
    alphaBoolBlend {blendOption=Linear}
    boneToModify {boneName=dyn_pants_l}
    translation=<-13 8>
   }
   n24 = "Transform (Modify) Bone - Bone: dyn_pants_r"(ComponentPose=@n23, Alpha=Get Calf Rot R(), Translation="13,-8,0", Rotation="0,0,0", Scale="1,1,1") {
    translationMode=BMM_Additive
    translationSpace=BCS_BoneSpace
    alphaBoolBlend {blendOption=Linear}
    boneToModify {boneName=dyn_pants_r}
    translation=<13 -8>
   }
   n25 = "Constraint - dyn_ropes_l_MSTR (PPPPP)"!(
    ComponentPose=@n24
    Alpha=1
    ConstraintWeights_0=0
    ConstraintWeights_1:float=0.0331
    ConstraintWeights_2:float=0.0108
    ConstraintWeights_3:float=0.4466
    ConstraintWeights_4:float=0.5096
   ) -> (Pose) {
    alphaBoolBlend {blendOption=Linear}
    boneToModify {boneName=dyn_ropes_l_MSTR}
    constraintSetup [
     {targetBone {boneName=pelvis}}
     {targetBone {boneName=thigh_l}}
     {targetBone {boneName=thigh_twist_01_l}}
     {targetBone {boneName=calf_twist_02_l}}
     {targetBone {boneName=dyn_pants_l}}
    ]
    constraintWeights [1,1,1,1,1]
   }
   n26 = "Constraint - dyn_ropes_r_MSTR (PPPPPP)"!(
    ComponentPose=@n25
    Alpha=1
    ConstraintWeights_0=0
    ConstraintWeights_1:float=0.0084
    ConstraintWeights_2:float=0.0032
    ConstraintWeights_3:float=0.0914
    ConstraintWeights_4:float=0.3492
    ConstraintWeights_5:float=0.5478
   ) -> (Pose) {
    alphaBoolBlend {blendOption=Linear}
    boneToModify {boneName=dyn_ropes_r_MSTR}
    constraintSetup [
     {targetBone {boneName=pelvis}}
     {targetBone {boneName=thigh_r}}
     {targetBone {boneName=thigh_twist_01_r}}
     {targetBone {boneName=calf_r}}
     {targetBone {boneName=calf_twist_02_r}}
     {targetBone {boneName=dyn_pants_r}}
    ]
    constraintWeights [1,1,1,1,1,1]
   }
   Save cached pose 'pre sim'(Pose=Component To Local(ComponentPose=@n26)) {cacheName=pre sim}
   Get thigh_rot_l()
  ]}
 EventGraph {nodes [
   Event BlueprintUpdateAnimation()
   > n1 = Sequence()
   n2 = Get Skel_Mesh()
   > Set thigh_rot_l(thigh_rot_l=GetAlphaFromReferenceTransformDelta(SkeletalMeshComponent=@n2, BoneName=thigh_l, BaseName=pelvis, InputRange=(X=-20,Y=-115), TwistAxis="1,0,0"))
   > Set thigh_rot_r(thigh_rot_r=GetAlphaFromReferenceTransformDelta(SkeletalMeshComponent=@n2, BoneName=thigh_r, BaseName=pelvis, InputRange=(X=-20,Y=-100), TwistAxis="1,0,0"))
   > Set Arm L rot(Arm L rot:double=GetAlphaFromReferenceTransformDelta(SkeletalMeshComponent=Get Skel_Mesh(), BoneName=upperarm_l, BaseName=spine_05, InputRange=(X=40,Y=90), TwistAxis="1,0,0")) -> (Output_Get:double)
   n3 = Get Skel_Mesh()
   > Set Calf Rot L(Calf Rot L:double=GetAlphaFromReferenceTransformDelta(SkeletalMeshComponent=@n3, BoneName=calf_l, BaseName=thigh_l, InputRange=(X=50,Y=110), TwistAxis="1,0,0")) -> (Output_Get:double)
   > Set Calf Rot R(Calf Rot R:double=GetAlphaFromReferenceTransformDelta(SkeletalMeshComponent=@n3, BoneName=calf_r, BaseName=thigh_r, InputRange=(X=50,Y=110), TwistAxis="1,0,0")) -> (Output_Get:double)
   n4 = Gravity Override State Function DEPRECATED(
    execute=@n1.then_1
    CharacterPartAnimInstance=Self-Reference()
    GameTimeInSeconds=GetGameTimeInSeconds()
    GravityOverrideParams=Get Gravity Override Params()
   )
   > Set GravityOverride(GravityOverride=@n4)
   Event BlueprintInitializeAnimation()
   > Set Skel_Mesh(Skel_Mesh=GetOwnerSkeletalMeshComponent())
   K2Node_Event_3 = CustomEvent() {is=title K2Node_CustomEvent}
   > Set Skel_Mesh(Skel_Mesh=GetOwnerSkeletalMeshComponent())
  ]}
]
variables [
 Arm L rot {type=number}
 Calf Rot L {type=number}
 Calf Rot R {type=number}
 Gravity Override Params {type=GravityOverrideParamsStruct}
 GravityOverride {type=Vector}
 Skel_Mesh {type=SkeletalMeshComponent}
 thigh_rot_l {type=number}
 thigh_rot_r {type=number}
]
