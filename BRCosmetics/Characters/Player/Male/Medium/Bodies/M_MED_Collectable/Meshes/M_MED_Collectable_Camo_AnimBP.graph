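# Animation Blueprint for the Camo variant of the male medium Collectable body cosmetic, posing its dyn_spine_05 and dyn_abdomen helper bones.
# It copies the attached parent mesh's pose, then constrains dyn_spine_05 to spine and head bones, with a slight translation pull toward both upper arms.
# dyn_abdomen takes partial rotation from that result, then Look At nodes aim both bones at their TGT bones.
# All work is in the AnimGraph, without physics or variables.
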AnimBlueprint /BRCosmetics/Characters/Player/Male/Medium/Bodies/M_MED_Collectable/Meshes/M_MED_Collectable_Camo_AnimBP
target_skeleton=&/M_MED_Collectable_Camo_Skeleton
graphs [
 AnimGraph {nodes [
   n1 = "Constraint - dyn_spine_05 (PPP)"!(
    ComponentPose=Local To Component(LocalPose=Copy Pose From Mesh() {bUseAttachedParent=true})
    Alpha=1
    ConstraintWeights_0=1
    ConstraintWeights_1:float=0
    ConstraintWeights_2:float=0
   ) -> (Pose) {
    alphaBoolBlend {blendOption=Linear}
    boneToModify {boneName=dyn_spine_05}
    constraintSetup [{targetBone {boneName=spine_05}}, {targetBone {boneName=spine_04}}, {targetBone {boneName=head}}]
    constraintWeights [1,1,1]
   }
   n2 = "Constraint - dyn_spine_05 (TT)"!(ComponentPose=@n1, Alpha=0.1, ConstraintWeights_0=0.6, ConstraintWeights_1:float=0.4) -> (Pose) {
    alphaBoolBlend {blendOption=Linear}
    boneToModify {boneName=dyn_spine_05}
    constraintSetup [{transformType=Translation, targetBone {boneName=upperarm_l}}, {transformType=Translation, targetBone {boneName=upperarm_r}}]
    constraintWeights [1,1]
   }
   n3 = "Constraint - dyn_abdomen (R)"(ComponentPose=@n2, Alpha=1, ConstraintWeights_0=0.5) {
    alphaBoolBlend {blendOption=Linear}
    boneToModify {boneName=dyn_abdomen}
    constraintSetup [{transformType=Rotation, targetBone {boneName=dyn_spine_05}}]
    constraintWeights [1]
   }
   n4 = Look At - Bone: dyn_spine_05(ComponentPose=@n3, Alpha=1) {
    alphaBoolBlend {blendOption=Linear}
    boneToModify {boneName=dyn_spine_05}
    lookAtLocation {x=0, z=1}
    lookAtTarget {boneReference {boneName=dyn_spine_TGT}}
    lookAt_Axis {axis=<1 0>}
   }
   n5 = Look At - Bone: dyn_abdomen(ComponentPose=@n4, Alpha=1) {
    alphaBoolBlend {blendOption=Linear}
    boneToModify {boneName=dyn_abdomen}
    lookAtLocation {x=0}
    lookAtTarget {boneReference {boneName=dyn_abdomen_TGT}}
    lookAt_Axis {axis=<1 0>}
   }
   Output Pose(Result=Component To Local(ComponentPose=@n5)) {layerGroup=None}
  ]}
 EventGraph {nodes [Event BlueprintUpdateAnimation()]}
]
variables []
